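# Animation Blueprint for the MillionaireGem back bling, posing its two feathered wings and adding secondary motion.
# Offsets on the input pose feed blends for running, skydiving and dive mode, with parachute and floating wing rotations.
# Chained cached poses then run spring-based AnimDynamics on feathers, plus mid-wing and end bones weighted by skydive mode.
# Finally, fake wind adds shared ambient rotators to left, then right base feathers, weighted by interpolated fakeMovement.
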AnimBlueprint /BRCosmetics/Accessories/FORT_Backpacks/Backpack_F_MED_MillionaireGem/Meshes/F_MED_MillionaireGem_Pack_AnimBP
target_skeleton=&/F_MED_MillionaireGem_Pack_Skeleton
graphs [
 AnimGraph {nodes [
   n1 = Get ambienRotatorB() -> (ambienRotatorB:Rotator)
   n2 = Get fakeMovement() -> (fakeMovement:double)
   n3 = "Transform (Modify) Bone - Bone: base_feather_01_r"(
    ComponentPose=Local To Component(LocalPose=Use cached pose 'CRCache_FakeWind_LSide'())
    Alpha=@n2
    Translation="0,0,0"
    Rotation=@n1
    Scale="1,1,1"
   ) {
    rotationMode=BMM_Additive
    rotationSpace=BCS_BoneSpace
    alphaBoolBlend {blendOption=Linear}
    alphaScaleBiasClamp {bInterpResult=true, interpSpeedIncreasing=1}
    boneToModify {boneName=base_feather_01_r}
   }
   n4 = Get ambientRotatorA() -> (ambientRotatorA:Rotator)
   n5 = "Transform (Modify) Bone - Bone: base_feather_02_r"(ComponentPose=@n3, Alpha=@n2, Translation="0,0,0", Rotation=@n4, Scale="1,1,1") {
    rotationMode=BMM_Additive
    rotationSpace=BCS_BoneSpace
    alphaBoolBlend {blendOption=Linear}
    alphaScaleBiasClamp {bInterpResult=true, interpSpeedIncreasing=1}
    boneToModify {boneName=base_feather_02_r}
   }
   n6 = Get ambientRotatorC() -> (ambientRotatorC:Rotator)
   n7 = "Transform (Modify) Bone - Bone: base_feather_03_r"(ComponentPose=@n5, Alpha=@n2, Translation="0,0,0", Rotation=@n6, Scale="1,1,1") {
    rotationMode=BMM_Additive
    rotationSpace=BCS_BoneSpace
    alphaBoolBlend {blendOption=Linear}
    alphaScaleBiasClamp {bInterpResult=true, interpSpeedIncreasing=1}
    boneToModify {boneName=base_feather_03_r}
   }
   n8 = Get ambientRotatorD() -> (ambientRotatorD:Rotator)
   n9 = "Transform (Modify) Bone - Bone: base_feather_04_r"(ComponentPose=@n7, Alpha=@n2, Translation="0,0,0", Rotation=@n8, Scale="1,1,1") {
    rotationMode=BMM_Additive
    rotationSpace=BCS_BoneSpace
    alphaBoolBlend {blendOption=Linear}
    alphaScaleBiasClamp {bInterpResult=true, interpSpeedIncreasing=1}
    boneToModify {boneName=base_feather_04_r}
   }
   n10 = Get ambientRotatorA()
   n11 = "Transform (Modify) Bone - Bone: base_feather_05_l"(ComponentPose=@n9, Alpha=@n2, Translation="0,0,0", Rotation=@n10, Scale="1,1,1") {
    rotationMode=BMM_Additive
    rotationSpace=BCS_BoneSpace
    alphaBoolBlend {blendOption=Linear}
    alphaScaleBiasClamp {bInterpResult=true, interpSpeedIncreasing=1}
    boneToModify {boneName=base_feather_05_l}
   }
   n12 = Get ambientRotatorC()
   n13 = "Transform (Modify) Bone - Bone: base_feather_06_r"(ComponentPose=@n11, Alpha=@n2, Translation="0,0,0", Rotation=@n12, Scale="1,1,1") {
    rotationMode=BMM_Additive
    rotationSpace=BCS_BoneSpace
    alphaBoolBlend {blendOption=Linear}
    alphaScaleBiasClamp {bInterpResult=true, interpSpeedIncreasing=1}
    boneToModify {boneName=base_feather_06_r}
   }
   n14 = "Transform (Modify) Bone - Bone: base_feather_07_r"(ComponentPose=@n13, Alpha=@n2, Translation="0,0,0", Rotation=@n1, Scale="1,1,1") {
    rotationMode=BMM_Additive
    rotationSpace=BCS_BoneSpace
    alphaBoolBlend {blendOption=Linear}
    alphaScaleBiasClamp {bInterpResult=true, interpSpeedIncreasing=1}
    boneToModify {boneName=base_feather_07_r}
   }
   n15 = "Transform (Modify) Bone - Bone: base_feather_08_r"(ComponentPose=@n14, Alpha=@n2, Translation="0,0,0", Rotation=@n4, Scale="1,1,1") {
    rotationMode=BMM_Additive
    rotationSpace=BCS_BoneSpace
    alphaBoolBlend {blendOption=Linear}
    alphaScaleBiasClamp {bInterpResult=true, interpSpeedIncreasing=1}
    boneToModify {boneName=base_feather_08_r}
   }
   n16 = "Transform (Modify) Bone - Bone: base_feather_09_r"(ComponentPose=@n15, Alpha=@n2, Translation="0,0,0", Rotation=@n6, Scale="1,1,1") {
    rotationMode=BMM_Additive
    rotationSpace=BCS_BoneSpace
    alphaBoolBlend {blendOption=Linear}
    alphaScaleBiasClamp {bInterpResult=true, interpSpeedIncreasing=1}
    boneToModify {boneName=base_feather_09_r}
   }
   n17 = "Transform (Modify) Bone - Bone: base_feather_10_r"(ComponentPose=@n16, Alpha=@n2, Translation="0,0,0", Rotation=@n8, Scale="1,1,1") {
    rotationMode=BMM_Additive
    rotationSpace=BCS_BoneSpace
    alphaBoolBlend {blendOption=Linear}
    alphaScaleBiasClamp {bInterpResult=true, interpSpeedIncreasing=1}
    boneToModify {boneName=base_feather_10_r}
   }
   n18 = "Transform (Modify) Bone - Bone: base_feather_11_r"(ComponentPose=@n17, Alpha=@n2, Translation="0,0,0", Rotation=@n10, Scale="1,1,1") {
    rotationMode=BMM_Additive
    rotationSpace=BCS_BoneSpace
    alphaBoolBlend {blendOption=Linear}
    alphaScaleBiasClamp {bInterpResult=true, interpSpeedIncreasing=1}
    boneToModify {boneName=base_feather_11_r}
   }
   n19 = "Transform (Modify) Bone - Bone: base_feather_12_r"(ComponentPose=@n18, Alpha=@n2, Translation="0,0,0", Rotation=@n12, Scale="1,1,1") {
    rotationMode=BMM_Additive
    rotationSpace=BCS_BoneSpace
    alphaBoolBlend {blendOption=Linear}
    alphaScaleBiasClamp {bInterpResult=true, interpSpeedIncreasing=1}
    boneToModify {boneName=base_feather_12_r}
   }
   n20 = "Transform (Modify) Bone - Bone: base_feather_13_r"(ComponentPose=@n19, Alpha=@n2, Translation="0,0,0", Rotation=@n1, Scale="1,1,1") {
    rotationMode=BMM_Additive
    rotationSpace=BCS_BoneSpace
    alphaBoolBlend {blendOption=Linear}
    alphaScaleBiasClamp {bInterpResult=true, interpSpeedIncreasing=1}
    boneToModify {boneName=base_feather_13_r}
   }
   Output Pose(Result=Component To Local(ComponentPose=@n20)) {layerGroup=None}
   n21 = "Transform (Modify) Bone - Bone: base_wing_a_l"(
    ComponentPose=Local To Component(LocalPose=Use cached pose 'CRCache_Entry'())
    Alpha=1
    Translation="0,0,0"
    Rotation="0,30,-10"
    Scale="1,1,1"
   ) {
    rotationMode=BMM_Additive
    rotationSpace=BCS_BoneSpace
    alphaBoolBlend {blendOption=Linear}
    boneToModify {boneName=base_wing_a_l}
    rotation {roll=-10, yaw=30}
   }
   n22 = "Transform (Modify) Bone - Bone: base_wing_a_r"(ComponentPose=@n21, Alpha=1, Translation="0,0,0", Rotation="0,30,-10", Scale="1,1,1") {
    rotationMode=BMM_Additive
    rotationSpace=BCS_BoneSpace
    alphaBoolBlend {blendOption=Linear}
    boneToModify {boneName=base_wing_a_r}
    rotation {roll=-10, yaw=30}
   }
   Save cached pose 'CRCache_runningStatus'(Pose=Blend Poses by bool(
     BlendPose_0=Component To Local(ComponentPose=@n22)
     BlendPose_1=Use cached pose 'CRCache_Entry'()
     BlendTime_0=0.4
     BlendTime_1=0.4
     bActiveValue=Get runningVar() -> (runningVar:bool)
    ) {blendType=Sinusoidal}) {cacheName=CRCache_runningStatus}
   n23 = "Transform (Modify) Bone - Bone: pack"(ComponentPose=Local To Component(LocalPose=Input Pose()), Translation="0,1.4,0", Rotation="0,0.01,0") {
    rotationMode=BMM_Additive
    rotationSpace=BCS_BoneSpace
    translationMode=BMM_Additive
    translationSpace=BCS_BoneSpace
    alphaBoolBlend {blendOption=Linear}
    boneToModify {boneName=pack}
    rotation {yaw=0.01}
    translation {y=1.4}
   }
   n24 = "Transform (Modify) Bone - Bone: base_wing_a_r"!(ComponentPose=@n23, Alpha=1, Rotation="0,0,0", Scale="1,1,1") -> (Pose) {
    rotationMode=BMM_Additive
    rotationSpace=BCS_BoneSpace
    translationMode=BMM_Additive
    translationSpace=BCS_ParentBoneSpace
    alphaBoolBlend {blendOption=Linear}
    boneToModify {boneName=base_wing_a_r}
    translation {y=8}
   }
   n25 = "Transform (Modify) Bone - Bone: base_wing_a_l"!(ComponentPose=@n24, Alpha=1, Rotation="0,0,0", Scale="1,1,1") -> (Pose) {
    rotationMode=BMM_Additive
    rotationSpace=BCS_BoneSpace
    translationMode=BMM_Additive
    translationSpace=BCS_ParentBoneSpace
    alphaBoolBlend {blendOption=Linear}
    boneToModify {boneName=base_wing_a_l}
    translation {y=8}
   }
   Save cached pose 'CRCache_Entry'(Pose=Component To Local(ComponentPose=@n25)) {cacheName=CRCache_Entry}
   n26 = Get ambienRotatorB()
   n27 = Get fakeMovement()
   n28 = "Transform (Modify) Bone - Bone: base_feather_01_l"(
    ComponentPose=Local To Component(LocalPose=Use cached pose 'CRCache_Dynamics'())
    Alpha=@n27
    Translation="0,0,0"
    Rotation=@n26
    Scale="1,1,1"
   ) {
    rotationMode=BMM_Additive
    rotationSpace=BCS_BoneSpace
    alphaBoolBlend {blendOption=Linear}
    alphaScaleBiasClamp {bInterpResult=true, interpSpeedDecreasing=1, interpSpeedIncreasing=20}
    boneToModify {boneName=base_feather_01_l}
   }
   n29 = Get ambientRotatorA()
   n30 = "Transform (Modify) Bone - Bone: base_feather_02_l"(ComponentPose=@n28, Alpha=@n27, Translation="0,0,0", Rotation=@n29, Scale="1,1,1") {
    rotationMode=BMM_Additive
    rotationSpace=BCS_BoneSpace
    alphaBoolBlend {blendOption=Linear}
    alphaScaleBiasClamp {bInterpResult=true, interpSpeedDecreasing=1, interpSpeedIncreasing=20}
    boneToModify {boneName=base_feather_02_l}
   }
   n31 = Get ambientRotatorC()
   n32 = "Transform (Modify) Bone - Bone: base_feather_03_l"(ComponentPose=@n30, Alpha=@n27, Translation="0,0,0", Rotation=@n31, Scale="1,1,1") {
    rotationMode=BMM_Additive
    rotationSpace=BCS_BoneSpace
    alphaBoolBlend {blendOption=Linear}
    alphaScaleBiasClamp {bInterpResult=true, interpSpeedDecreasing=1, interpSpeedIncreasing=20}
    boneToModify {boneName=base_feather_03_l}
   }
   n33 = Get ambientRotatorD()
   n34 = "Transform (Modify) Bone - Bone: base_feather_04_l"(ComponentPose=@n32, Alpha=@n27, Translation="0,0,0", Rotation=@n33, Scale="1,1,1") {
    rotationMode=BMM_Additive
    rotationSpace=BCS_BoneSpace
    alphaBoolBlend {blendOption=Linear}
    alphaScaleBiasClamp {bInterpResult=true, interpSpeedDecreasing=1, interpSpeedIncreasing=20}
    boneToModify {boneName=base_feather_04_l}
   }
   n35 = Get ambientRotatorA()
   n36 = "Transform (Modify) Bone - Bone: base_feather_05_l"(ComponentPose=@n34, Alpha=@n27, Translation="0,0,0", Rotation=@n35, Scale="1,1,1") {
    rotationMode=BMM_Additive
    rotationSpace=BCS_BoneSpace
    alphaBoolBlend {blendOption=Linear}
    alphaScaleBiasClamp {bInterpResult=true, interpSpeedDecreasing=1, interpSpeedIncreasing=20}
    boneToModify {boneName=base_feather_05_l}
   }
   n37 = Get ambientRotatorC()
   n38 = "Transform (Modify) Bone - Bone: base_feather_06_l"(ComponentPose=@n36, Alpha=@n27, Translation="0,0,0", Rotation=@n37, Scale="1,1,1") {
    rotationMode=BMM_Additive
    rotationSpace=BCS_BoneSpace
    alphaBoolBlend {blendOption=Linear}
    alphaScaleBiasClamp {bInterpResult=true, interpSpeedDecreasing=1, interpSpeedIncreasing=20}
    boneToModify {boneName=base_feather_06_l}
   }
   n39 = "Transform (Modify) Bone - Bone: base_feather_07_l"(ComponentPose=@n38, Alpha=@n27, Translation="0,0,0", Rotation=@n26, Scale="1,1,1") {
    rotationMode=BMM_Additive
    rotationSpace=BCS_BoneSpace
    alphaBoolBlend {blendOption=Linear}
    alphaScaleBiasClamp {bInterpResult=true, interpSpeedDecreasing=1, interpSpeedIncreasing=20}
    boneToModify {boneName=base_feather_07_l}
   }
   n40 = "Transform (Modify) Bone - Bone: base_feather_08_l"(ComponentPose=@n39, Alpha=@n27, Translation="0,0,0", Rotation=@n29, Scale="1,1,1") {
    rotationMode=BMM_Additive
    rotationSpace=BCS_BoneSpace
    alphaBoolBlend {blendOption=Linear}
    alphaScaleBiasClamp {bInterpResult=true, interpSpeedDecreasing=1, interpSpeedIncreasing=20}
    boneToModify {boneName=base_feather_08_l}
   }
   n41 = "Transform (Modify) Bone - Bone: base_feather_09_l"(ComponentPose=@n40, Alpha=@n27, Translation="0,0,0", Rotation=@n31, Scale="1,1,1") {
    rotationMode=BMM_Additive
    rotationSpace=BCS_BoneSpace
    alphaBoolBlend {blendOption=Linear}
    alphaScaleBiasClamp {bInterpResult=true, interpSpeedDecreasing=1, interpSpeedIncreasing=20}
    boneToModify {boneName=base_feather_09_l}
   }
   n42 = "Transform (Modify) Bone - Bone: base_feather_10_l"(ComponentPose=@n41, Alpha=@n27, Translation="0,0,0", Rotation=@n33, Scale="1,1,1") {
    rotationMode=BMM_Additive
    rotationSpace=BCS_BoneSpace
    alphaBoolBlend {blendOption=Linear}
    alphaScaleBiasClamp {bInterpResult=true, interpSpeedDecreasing=1, interpSpeedIncreasing=20}
    boneToModify {boneName=base_feather_10_l}
   }
   n43 = "Transform (Modify) Bone - Bone: base_feather_11_l"(ComponentPose=@n42, Alpha=@n27, Translation="0,0,0", Rotation=@n35, Scale="1,1,1") {
    rotationMode=BMM_Additive
    rotationSpace=BCS_BoneSpace
    alphaBoolBlend {blendOption=Linear}
    alphaScaleBiasClamp {bInterpResult=true, interpSpeedDecreasing=1, interpSpeedIncreasing=20}
    boneToModify {boneName=base_feather_11_l}
   }
   n44 = "Transform (Modify) Bone - Bone: base_feather_12_l"(ComponentPose=@n43, Alpha=@n27, Translation="0,0,0", Rotation=@n37, Scale="1,1,1") {
    rotationMode=BMM_Additive
    rotationSpace=BCS_BoneSpace
    alphaBoolBlend {blendOption=Linear}
    alphaScaleBiasClamp {bInterpResult=true, interpSpeedDecreasing=1, interpSpeedIncreasing=20}
    boneToModify {boneName=base_feather_12_l}
   }
   n45 = "Transform (Modify) Bone - Bone: base_feather_13_l"(ComponentPose=@n44, Alpha=@n27, Translation="0,0,0", Rotation=@n26, Scale="1,1,1") {
    rotationMode=BMM_Additive
    rotationSpace=BCS_BoneSpace
    alphaBoolBlend {blendOption=Linear}
    alphaScaleBiasClamp {bInterpResult=true, interpSpeedDecreasing=1, interpSpeedIncreasing=20}
    boneToModify {boneName=base_feather_13_l}
   }
   Save cached pose 'CRCache_FakeWind_LSide'(Pose=Component To Local(ComponentPose=@n45)) {cacheName=CRCache_FakeWind_LSide}
   n46 = Get parachuteOpen() -> (parachuteOpen:double)
   n47 = "Transform (Modify) Bone - Bone: base_wing_a_l"!(
    ComponentPose=Local To Component(LocalPose=Use cached pose 'CRCache_runningStatus'())
    Alpha=@n46
    Translation:Vector="0,0,0"
    Rotation="-7,2,0"
    Scale="1,1,1"
   ) -> (Pose) {
    rotationMode=BMM_Additive
    rotationSpace=BCS_BoneSpace
    alphaBoolBlend {blendOption=Linear}
    alphaScaleBiasClamp {bInterpResult=true}
    boneToModify {boneName=base_wing_a_l}
    rotation {pitch=-7, yaw=2}
   }
   n48 = "Transform (Modify) Bone - Bone: base_wing_a_r"!(ComponentPose=@n47, Alpha=@n46, Translation:Vector="0,0,0", Rotation="-7,2,0", Scale="1,1,1") -> (Pose) {
    rotationMode=BMM_Additive
    rotationSpace=BCS_BoneSpace
    alphaBoolBlend {blendOption=Linear}
    alphaScaleBiasClamp {bInterpResult=true}
    boneToModify {boneName=base_wing_a_r}
    rotation {pitch=-7, yaw=2}
   }
   n49 = Get WingsRotation() -> (WingsRotation:Rotator)
   n50 = Get SkydiveFloatingAlpha() -> (SkydiveFloatingAlpha:float)
   n51 = "Transform (Modify) Bone - Bone: base_wing_a_l"(ComponentPose=@n48, Alpha=@n50, Translation="0,0,0", Rotation=@n49, Scale="1,1,1") {
    rotationMode=BMM_Additive
    rotationSpace=BCS_BoneSpace
    alphaBoolBlend {blendOption=Linear}
    alphaScaleBiasClamp {bInterpResult=true}
    boneToModify {boneName=base_wing_a_l}
   }
   n52 = "Transform (Modify) Bone - Bone: base_wing_a_r"(ComponentPose=@n51, Alpha=@n50, Translation="0,0,0", Rotation=@n49, Scale="1,1,1") {
    rotationMode=BMM_Additive
    rotationSpace=BCS_BoneSpace
    alphaBoolBlend {blendOption=Linear}
    alphaScaleBiasClamp {bInterpResult=true}
    boneToModify {boneName=base_wing_a_r}
   }
   Save cached pose 'CRCache_floatingSky'(Pose=Blend Poses by bool(
     BlendPose_0=Component To Local(ComponentPose=@n52)
     BlendPose_1=Use cached pose 'CRCache_runningStatus'()
     BlendTime_0=0.3
     BlendTime_1=0.3
     bActiveValue=Get skyDive() -> (skyDive:bool)
    ) {blendType=CubicInOut}) {cacheName=CRCache_floatingSky}
   n53 = Get GravityOverride()
   n54 = Get skyDiveMode() -> (skyDiveMode:double)
   n55 = Anim Dynamics - Bone: dyn_wingmid_01_r!(
    ComponentPose=Local To Component(LocalPose=Use cached pose 'CRCache_Large_feathers'())
    Alpha=@n54
    GravityOverride:Vector=@n53
    ExternalForce="0,0,0"
   ) -> (Pose) {
    angularDampingOverride=0.999
    angularSpringConstant=20000
    bAngularSpring=true
    bOverrideAngularDamping=true
    bOverrideLinearDamping=true
    bUseGravityOverride=true
    gravityScale=0.35
    linearDampingOverride=0.999
    simulationSpace=RootRelative
    alphaBoolBlend {blendOption=Linear}
    alphaScaleBias {scale=0.5}
    boundBone {boneName=dyn_wingmid_01_r}
    componentAppliedLinearAccClamp=<3000 3000 3000>
    componentLinearAccScale=<0.2 0.2 0.2>
    componentLinearVelScale=<0.2 0.2 0.2>
    gravityOverride {z=-980}
    physicsBodyDefinitions [{
      boundBone {boneName=dyn_wingmid_01_r}
      boxExtents=<10 10 10>
      constraintSetup {angularLimitsMax=<0 0.5 0>, angularLimitsMin=<0 -35 0>, angularTarget=<1 0 0>}
      localJointOffset=<5 0 0>
     }]
   }
   n56 = Anim Dynamics - Bone: dyn_wingmid_02_r!(ComponentPose=@n55, Alpha=@n54, GravityOverride:Vector=@n53, ExternalForce="0,0,0") -> (Pose) {
    angularDampingOverride=0.999
    angularSpringConstant=20000
    bAngularSpring=true
    bOverrideAngularDamping=true
    bOverrideLinearDamping=true
    bUseGravityOverride=true
    gravityScale=0.35
    linearDampingOverride=0.999
    simulationSpace=RootRelative
    alphaBoolBlend {blendOption=Linear}
    alphaScaleBias {scale=0.5}
    boundBone {boneName=dyn_wingmid_02_r}
    componentAppliedLinearAccClamp=<3000 3000 3000>
    componentLinearAccScale=<0.2 0.2 0.2>
    componentLinearVelScale=<0.2 0.2 0.2>
    gravityOverride {z=-980}
    physicsBodyDefinitions [{
      boundBone {boneName=dyn_wingmid_02_r}
      boxExtents=<10 10 10>
      constraintSetup {angularLimitsMax=<0 0.5 0>, angularLimitsMin=<0 -35 0>, angularTarget=<1 0 0>}
      localJointOffset=<5 0 0>
     }]
   }
   n57 = Anim Dynamics - Bone: dyn_wingmid_01_l!(ComponentPose=@n56, Alpha=@n54, GravityOverride:Vector=@n53, ExternalForce="0,0,0") -> (Pose) {
    angularDampingOverride=0.999
    angularSpringConstant=20000
    bAngularSpring=true
    bOverrideAngularDamping=true
    bOverrideLinearDamping=true
    bUseGravityOverride=true
    gravityScale=0.35
    linearDampingOverride=0.999
    simulationSpace=RootRelative
    alphaBoolBlend {blendOption=Linear}
    alphaScaleBias {scale=0.5}
    boundBone {boneName=dyn_wingmid_01_l}
    componentAppliedLinearAccClamp=<3000 3000 3000>
    componentLinearAccScale=<0.2 0.2 0.2>
    componentLinearVelScale=<0.2 0.2 0.2>
    gravityOverride {z=-980}
    physicsBodyDefinitions [{
      boundBone {boneName=dyn_wingmid_01_l}
      boxExtents=<10 10 10>
      constraintSetup {angularLimitsMax=<0 0.5 0>, angularLimitsMin=<0 -35 0>, angularTarget=<1 0 0>}
      localJointOffset=<5 0 0>
     }]
   }
   n58 = Anim Dynamics - Bone: dyn_wingmid_02_l!(ComponentPose=@n57, Alpha=@n54, GravityOverride:Vector=@n53, ExternalForce="0,0,0") -> (Pose) {
    angularDampingOverride=0.999
    angularSpringConstant=20000
    bAngularSpring=true
    bOverrideAngularDamping=true
    bOverrideLinearDamping=true
    bUseGravityOverride=true
    gravityScale=0.35
    linearDampingOverride=0.999
    simulationSpace=RootRelative
    alphaBoolBlend {blendOption=Linear}
    alphaScaleBias {scale=0.5}
    boundBone {boneName=dyn_wingmid_02_l}
    componentAppliedLinearAccClamp=<3000 3000 3000>
    componentLinearAccScale=<0.2 0.2 0.2>
    componentLinearVelScale=<0.2 0.2 0.2>
    gravityOverride {z=-980}
    physicsBodyDefinitions [{
      boundBone {boneName=dyn_wingmid_02_l}
      boxExtents=<10 10 10>
      constraintSetup {angularLimitsMax=<0 0.5 0>, angularLimitsMin=<0 -35 0>, angularTarget=<1 0 0>}
      localJointOffset=<5 0 0>
     }]
   }
   n59 = Anim Dynamics - Bone: dyn_windend_01_r!(ComponentPose=@n58, Alpha=@n54, GravityOverride:Vector=@n53, ExternalForce="0,0,0") -> (Pose) {
    angularDampingOverride=0.9
    angularSpringConstant=20000
    bAngularSpring=true
    bOverrideAngularDamping=true
    bOverrideLinearDamping=true
    bUseGravityOverride=true
    gravityScale=0.35
    linearDampingOverride=0.999
    simulationSpace=RootRelative
    alphaBoolBlend {blendOption=Linear}
    alphaScaleBias {scale=0.5}
    boundBone {boneName=dyn_windend_01_r}
    componentAppliedLinearAccClamp=<3000 3000 3000>
    componentLinearAccScale=<0.3 0.3 0.3>
    componentLinearVelScale=<0.3 0.3 0.3>
    gravityOverride {z=-980}
    physicsBodyDefinitions [{
      boundBone {boneName=dyn_windend_01_r}
      boxExtents=<10 10 10>
      constraintSetup {angularLimitsMax=<0 0.5 0>, angularLimitsMin=<0 -40 0>, angularTarget=<1 0 0>}
      localJointOffset=<3 0 0>
     }]
   }
   n60 = Anim Dynamics - Bone: dyn_windend_01_l!(ComponentPose=@n59, Alpha=@n54, GravityOverride:Vector=@n53, ExternalForce="0,0,0") -> (Pose) {
    angularDampingOverride=0.9
    angularSpringConstant=20000
    bAngularSpring=true
    bOverrideAngularDamping=true
    bOverrideLinearDamping=true
    bUseGravityOverride=true
    gravityScale=0.35
    linearDampingOverride=0.999
    simulationSpace=RootRelative
    alphaBoolBlend {blendOption=Linear}
    alphaScaleBias {scale=0.5}
    boundBone {boneName=dyn_windend_01_l}
    componentAppliedLinearAccClamp=<3000 3000 3000>
    componentLinearAccScale=<0.3 0.3 0.3>
    componentLinearVelScale=<0.3 0.3 0.3>
    gravityOverride {z=-980}
    physicsBodyDefinitions [{
      boundBone {boneName=dyn_windend_01_l}
      boxExtents=<10 10 10>
      constraintSetup {angularLimitsMax=<0 0.5 0>, angularLimitsMin=<0 -40 0>, angularTarget=<1 0 0>}
      localJointOffset=<3 0 0>
     }]
   }
   Save cached pose 'CRCache_Dynamics'(Pose=Component To Local(ComponentPose=@n60)) {cacheName=CRCache_Dynamics}
   n61 = "Transform (Modify) Bone - Bone: base_wing_a_l"(
    ComponentPose=Local To Component(LocalPose=Use cached pose 'CRCache_floatingSky'())
    Alpha=1
    Translation="0,0,0"
    Rotation="10,37,-25"
    Scale="1,1,1"
   ) {
    rotationMode=BMM_Additive
    rotationSpace=BCS_BoneSpace
    alphaBoolBlend {blendOption=Linear}
    boneToModify {boneName=base_wing_a_l}
    rotation=<rot 10 37 -25>
   }
   n62 = "Transform (Modify) Bone - Bone: base_wing_a_r"(ComponentPose=@n61, Alpha=1, Translation="0,0,0", Rotation="10,37,-25", Scale="1,1,1") {
    rotationMode=BMM_Additive
    rotationSpace=BCS_BoneSpace
    alphaBoolBlend {blendOption=Linear}
    boneToModify {boneName=base_wing_a_r}
    rotation=<rot 10 37 -25>
   }
   Save cached pose 'CRCache_AfterTransforms'(Pose=Blend Poses by bool(
     BlendPose_0=Component To Local(ComponentPose=@n62)
     BlendPose_1=Use cached pose 'CRCache_floatingSky'()
     BlendTime_0=0.3
     BlendTime_1=0.5
     bActiveValue=Get bIsSkydiveDiveMode()
    ) {blendType=CubicInOut}) {cacheName=CRCache_AfterTransforms}
   n63 = Anim Dynamics - Bone: dyn_feather_02_l(ComponentPose=Local To Component(LocalPose=Use cached pose 'CRCache_AfterTransforms'()), Alpha=0.95, ExternalForce="0,0,0") {
    angularDampingOverride=1
    angularSpringConstant=20000
    bAngularSpring=true
    bOverrideAngularDamping=true
    bOverrideLinearDamping=true
    bUseGravityOverride=true
    gravityScale=0.35
    linearDampingOverride=0.999
    numSolverIterationsPostUpdate=2
    simulationSpace=RootRelative
    alphaBoolBlend {blendOption=Linear}
    boundBone {boneName=dyn_feather_02_l}
    componentAppliedLinearAccClamp=<3000 3000 3000>
    componentLinearAccScale=<0.5 0.5 0.5>
    componentLinearVelScale=<0.5 0.5 0.5>
    gravityOverride {z=-980}
    physicsBodyDefinitions [{
      boundBone {boneName=dyn_feather_02_l}
      boxExtents=<20 20 20>
      constraintSetup {angularLimitsMax=<0 15 0>, angularLimitsMin=<0 -15 0>, angularTarget=<1 0 0>}
      localJointOffset=<2 0 0>
     }]
   }
   n64 = Anim Dynamics - Bone: dyn_feather_01_l(ComponentPose=@n63, Alpha=0.95, ExternalForce="0,0,0") {
    angularDampingOverride=1
    angularSpringConstant=20000
    bAngularSpring=true
    bOverrideAngularDamping=true
    bOverrideLinearDamping=true
    bUseGravityOverride=true
    gravityScale=0.35
    linearDampingOverride=0.999
    numSolverIterationsPostUpdate=2
    simulationSpace=RootRelative
    alphaBoolBlend {blendOption=Linear}
    boundBone {boneName=dyn_feather_01_l}
    componentAppliedLinearAccClamp=<3000 3000 3000>
    componentLinearAccScale=<0.5 0.5 0.5>
    componentLinearVelScale=<0.5 0.5 0.5>
    gravityOverride {z=-980}
    physicsBodyDefinitions [{
      boundBone {boneName=dyn_feather_01_l}
      boxExtents=<15 15 15>
      constraintSetup {angularLimitsMax=<0 15 0>, angularLimitsMin=<0 -15 0>, angularTarget=<1 0 0>}
      localJointOffset=<1 0 0>
     }]
   }
   n65 = Anim Dynamics - Bone: dyn_feather_01_r(ComponentPose=@n64, Alpha=0.95, ExternalForce="0,0,0") {
    angularDampingOverride=1
    angularSpringConstant=20000
    bAngularSpring=true
    bOverrideAngularDamping=true
    bOverrideLinearDamping=true
    bUseGravityOverride=true
    gravityScale=0.35
    linearDampingOverride=0.999
    numSolverIterationsPostUpdate=2
    simulationSpace=RootRelative
    alphaBoolBlend {blendOption=Linear}
    boundBone {boneName=dyn_feather_01_r}
    componentAppliedLinearAccClamp=<3000 3000 3000>
    componentLinearAccScale=<0.5 0.5 0.5>
    componentLinearVelScale=<0.5 0.5 0.5>
    gravityOverride {z=-980}
    physicsBodyDefinitions [{
      boundBone {boneName=dyn_feather_01_r}
      boxExtents=<20 20 20>
      constraintSetup {angularLimitsMax=<0 15 0>, angularLimitsMin=<0 -15 0>, angularTarget=<1 0 0>}
      localJointOffset=<2 0 0>
     }]
   }
   n66 = Anim Dynamics - Bone: dyn_feather_02_r(ComponentPose=@n65, Alpha=0.95, ExternalForce="0,0,0") {
    angularDampingOverride=1
    angularSpringConstant=20000
    bAngularSpring=true
    bOverrideAngularDamping=true
    bOverrideLinearDamping=true
    bUseGravityOverride=true
    gravityScale=0.35
    linearDampingOverride=0.999
    numSolverIterationsPostUpdate=2
    simulationSpace=RootRelative
    alphaBoolBlend {blendOption=Linear}
    boundBone {boneName=dyn_feather_02_r}
    componentAppliedLinearAccClamp=<3000 3000 3000>
    componentLinearAccScale=<0.5 0.5 0.5>
    componentLinearVelScale=<0.5 0.5 0.5>
    gravityOverride {z=-980}
    physicsBodyDefinitions [{
      boundBone {boneName=dyn_feather_02_r}
      boxExtents=<15 15 15>
      constraintSetup {angularLimitsMax=<0 15 0>, angularLimitsMin=<0 -15 0>, angularTarget=<1 0 0>}
      localJointOffset=<1 0 0>
     }]
   }
   Save cached pose 'CRCache_SmallFeathers'(Pose=Component To Local(ComponentPose=@n66)) {cacheName=CRCache_SmallFeathers}
   n67 = Get GravityOverride()
   n68 = Anim Dynamics - Bone: dyn_feather_03_r!(
    ComponentPose=Local To Component(LocalPose=Use cached pose 'CRCache_SmallFeathers'())
    Alpha=0.95
    GravityOverride:Vector=@n67
    ExternalForce="0,0,0"
   ) -> (Pose) {
    angularDampingOverride=1
    angularSpringConstant=20000
    bAngularSpring=true
    bOverrideAngularDamping=true
    bOverrideLinearDamping=true
    bUseGravityOverride=true
    gravityScale=0.35
    linearDampingOverride=0.95
    simulationSpace=RootRelative
    alphaBoolBlend {blendOption=Linear}
    alphaScaleBias {scale=0.5}
    boundBone {boneName=dyn_feather_03_r}
    componentAppliedLinearAccClamp=<3000 3000 3000>
    componentLinearAccScale=<0.4 0.4 0.4>
    componentLinearVelScale=<0.4 0.4 0.4>
    gravityOverride {z=-980}
    physicsBodyDefinitions [{
      boundBone {boneName=dyn_feather_03_r}
      boxExtents=<15 15 15>
      constraintSetup {angularLimitsMax=<0 15 0>, angularLimitsMin=<0 -15 0>, angularTarget=<1 0 0>}
      localJointOffset=<1.5 0 0>
     }]
   }
   n69 = Anim Dynamics - Bone: dyn_feather_04_r!(ComponentPose=@n68, Alpha=0.95, GravityOverride:Vector=@n67, ExternalForce="0,0,0") -> (Pose) {
    angularDampingOverride=1
    angularSpringConstant=20000
    bAngularSpring=true
    bOverrideAngularDamping=true
    bOverrideLinearDamping=true
    bUseGravityOverride=true
    gravityScale=0.35
    linearDampingOverride=0.95
    simulationSpace=RootRelative
    alphaBoolBlend {blendOption=Linear}
    alphaScaleBias {scale=0.5}
    boundBone {boneName=dyn_feather_04_r}
    componentAppliedLinearAccClamp=<3000 3000 3000>
    componentLinearAccScale=<0.4 0.4 0.4>
    componentLinearVelScale=<0.4 0.4 0.4>
    gravityOverride {z=-980}
    physicsBodyDefinitions [{
      boundBone {boneName=dyn_feather_04_r}
      boxExtents=<10 10 10>
      constraintSetup {angularLimitsMax=<0 10 0>, angularLimitsMin=<0 -5 0>, angularTarget=<1 0 0>}
      localJointOffset=<2 0 0>
     }]
   }
   n70 = Anim Dynamics - Bone: dyn_feather_05_r!(ComponentPose=@n69, Alpha=0.95, GravityOverride:Vector=@n67, ExternalForce="0,0,0") -> (Pose) {
    angularDampingOverride=1
    angularSpringConstant=20000
    bAngularSpring=true
    bOverrideAngularDamping=true
    bOverrideLinearDamping=true
    bUseGravityOverride=true
    gravityScale=0.35
    linearDampingOverride=0.95
    simulationSpace=RootRelative
    alphaBoolBlend {blendOption=Linear}
    alphaScaleBias {scale=0.5}
    boundBone {boneName=dyn_feather_05_r}
    componentAppliedLinearAccClamp=<3000 3000 3000>
    componentLinearAccScale=<0.4 0.4 0.4>
    componentLinearVelScale=<0.4 0.4 0.4>
    gravityOverride {z=-980}
    physicsBodyDefinitions [{
      boundBone {boneName=dyn_feather_05_r}
      boxExtents=<15 12 10>
      constraintSetup {angularLimitsMax=<0 5 0>, angularLimitsMin=<0 -10 0>, angularTarget=<1 0 0>}
      localJointOffset=<2.5 0 0>
     }]
   }
   n71 = Anim Dynamics - Bone: dyn_feather_06_r!(ComponentPose=@n70, Alpha=0.95, GravityOverride:Vector=@n67, ExternalForce="0,0,0") -> (Pose) {
    angularDampingOverride=1
    angularSpringConstant=20000
    bAngularSpring=true
    bOverrideAngularDamping=true
    bOverrideLinearDamping=true
    bUseGravityOverride=true
    gravityScale=0.35
    linearDampingOverride=0.95
    simulationSpace=RootRelative
    alphaBoolBlend {blendOption=Linear}
    alphaScaleBias {scale=0.5}
    boundBone {boneName=dyn_feather_06_r}
    componentAppliedLinearAccClamp=<3000 3000 3000>
    componentLinearAccScale=<0.4 0.4 0.4>
    componentLinearVelScale=<0.4 0.4 0.4>
    gravityOverride {z=-980}
    physicsBodyDefinitions [{
      boundBone {boneName=dyn_feather_06_r}
      boxExtents=<10 18 10>
      constraintSetup {angularLimitsMax=<0 8 0>, angularLimitsMin=<0 -8 0>, angularTarget=<1 0 0>}
      localJointOffset=<3 0 0>
     }]
   }
   n72 = Anim Dynamics - Bone: dyn_feather_07_r!(ComponentPose=@n71, Alpha=0.95, GravityOverride:Vector=@n67, ExternalForce="0,0,0") -> (Pose) {
    angularDampingOverride=1
    angularSpringConstant=20000
    bAngularSpring=true
    bOverrideAngularDamping=true
    bOverrideLinearDamping=true
    bUseGravityOverride=true
    gravityScale=0.35
    linearDampingOverride=0.95
    simulationSpace=RootRelative
    alphaBoolBlend {blendOption=Linear}
    alphaScaleBias {scale=0.5}
    boundBone {boneName=dyn_feather_07_r}
    componentAppliedLinearAccClamp=<3000 3000 3000>
    componentLinearAccScale=<0.4 0.4 0.4>
    componentLinearVelScale=<0.4 0.4 0.4>
    gravityOverride {z=-980}
    physicsBodyDefinitions [{
      boundBone {boneName=dyn_feather_07_r}
      boxExtents=<20 30 30>
      constraintSetup {angularLimitsMax=<0 15 0>, angularLimitsMin=<0 -15 0>, angularTarget=<1 0 0>}
      localJointOffset=<3 0 0>
     }]
   }
   n73 = Anim Dynamics - Bone: dyn_feather_03_l!(ComponentPose=@n72, Alpha=0.95, GravityOverride:Vector=@n67, ExternalForce="0,0,0") -> (Pose) {
    angularDampingOverride=1
    angularSpringConstant=20000
    bAngularSpring=true
    bOverrideAngularDamping=true
    bOverrideLinearDamping=true
    bUseGravityOverride=true
    gravityScale=0.35
    linearDampingOverride=0.95
    simulationSpace=RootRelative
    alphaBoolBlend {blendOption=Linear}
    alphaScaleBias {scale=0.5}
    boundBone {boneName=dyn_feather_03_l}
    componentAppliedLinearAccClamp=<3000 3000 3000>
    componentLinearAccScale=<0.4 0.4 0.4>
    componentLinearVelScale=<0.4 0.4 0.4>
    gravityOverride {z=-980}
    physicsBodyDefinitions [{
      boundBone {boneName=dyn_feather_03_l}
      boxExtents=<15 15 15>
      constraintSetup {angularLimitsMax=<0 15 0>, angularLimitsMin=<0 -15 0>, angularTarget=<1 0 0>}
      localJointOffset=<1.5 0 0>
     }]
   }
   n74 = Anim Dynamics - Bone: dyn_feather_04_l!(ComponentPose=@n73, Alpha=0.95, GravityOverride:Vector=@n67, ExternalForce="0,0,0") -> (Pose) {
    angularDampingOverride=1
    angularSpringConstant=20000
    bAngularSpring=true
    bOverrideAngularDamping=true
    bOverrideLinearDamping=true
    bUseGravityOverride=true
    gravityScale=0.35
    linearDampingOverride=0.95
    simulationSpace=RootRelative
    alphaBoolBlend {blendOption=Linear}
    alphaScaleBias {scale=0.5}
    boundBone {boneName=dyn_feather_04_l}
    componentAppliedLinearAccClamp=<3000 3000 3000>
    componentLinearAccScale=<0.4 0.4 0.4>
    componentLinearVelScale=<0.4 0.4 0.4>
    gravityOverride {z=-980}
    physicsBodyDefinitions [{
      boundBone {boneName=dyn_feather_04_l}
      boxExtents=<10 10 10>
      constraintSetup {angularLimitsMax=<0 10 0>, angularLimitsMin=<0 -5 0>, angularTarget=<1 0 0>}
      localJointOffset=<2 0 0>
     }]
   }
   n75 = Anim Dynamics - Bone: dyn_feather_05_l!(ComponentPose=@n74, Alpha=0.95, GravityOverride:Vector=@n67, ExternalForce="0,0,0") -> (Pose) {
    angularDampingOverride=1
    angularSpringConstant=20000
    bAngularSpring=true
    bOverrideAngularDamping=true
    bOverrideLinearDamping=true
    bUseGravityOverride=true
    gravityScale=0.35
    linearDampingOverride=0.95
    simulationSpace=RootRelative
    alphaBoolBlend {blendOption=Linear}
    alphaScaleBias {scale=0.5}
    boundBone {boneName=dyn_feather_05_l}
    componentAppliedLinearAccClamp=<3000 3000 3000>
    componentLinearAccScale=<0.4 0.4 0.4>
    componentLinearVelScale=<0.4 0.4 0.4>
    gravityOverride {z=-980}
    physicsBodyDefinitions [{
      boundBone {boneName=dyn_feather_05_l}
      boxExtents=<15 12 10>
      constraintSetup {angularLimitsMax=<0 5 0>, angularLimitsMin=<0 -10 0>, angularTarget=<1 0 0>}
      localJointOffset=<2.5 0 0>
     }]
   }
   n76 = Anim Dynamics - Bone: dyn_feather_06_l!(ComponentPose=@n75, Alpha=0.95, GravityOverride:Vector=@n67, ExternalForce="0,0,0") -> (Pose) {
    angularDampingOverride=1
    angularSpringConstant=20000
    bAngularSpring=true
    bOverrideAngularDamping=true
    bOverrideLinearDamping=true
    bUseGravityOverride=true
    gravityScale=0.35
    linearDampingOverride=0.95
    simulationSpace=RootRelative
    alphaBoolBlend {blendOption=Linear}
    alphaScaleBias {scale=0.5}
    boundBone {boneName=dyn_feather_06_l}
    componentAppliedLinearAccClamp=<3000 3000 3000>
    componentLinearAccScale=<0.4 0.4 0.4>
    componentLinearVelScale=<0.4 0.4 0.4>
    gravityOverride {z=-980}
    physicsBodyDefinitions [{
      boundBone {boneName=dyn_feather_06_l}
      boxExtents=<10 18 10>
      constraintSetup {angularLimitsMax=<0 8 0>, angularLimitsMin=<0 -8 0>, angularTarget=<1 0 0>}
      localJointOffset=<3 0 0>
     }]
   }
   n77 = Anim Dynamics - Bone: dyn_feather_07_l!(ComponentPose=@n76, Alpha=0.95, GravityOverride:Vector=@n67, ExternalForce="0,0,0") -> (Pose) {
    angularDampingOverride=1
    angularSpringConstant=20000
    bAngularSpring=true
    bOverrideAngularDamping=true
    bOverrideLinearDamping=true
    bUseGravityOverride=true
    gravityScale=0.35
    linearDampingOverride=0.95
    simulationSpace=RootRelative
    alphaBoolBlend {blendOption=Linear}
    alphaScaleBias {scale=0.5}
    boundBone {boneName=dyn_feather_07_l}
    componentAppliedLinearAccClamp=<3000 3000 3000>
    componentLinearAccScale=<0.4 0.4 0.4>
    componentLinearVelScale=<0.4 0.4 0.4>
    gravityOverride {z=-980}
    physicsBodyDefinitions [{
      boundBone {boneName=dyn_feather_07_l}
      boxExtents=<20 30 30>
      constraintSetup {angularLimitsMax=<0 15 0>, angularLimitsMin=<0 -15 0>, angularTarget=<1 0 0>}
      localJointOffset=<3 0 0>
     }]
   }
   Save cached pose 'CRCache_Mid_feathers'(Pose=Component To Local(ComponentPose=@n77)) {cacheName=CRCache_Mid_feathers}
   n78 = Get GravityOverride()
   n79 = Anim Dynamics - Bone: dyn_feather_08_r!(
    ComponentPose=Local To Component(LocalPose=Use cached pose 'CRCache_Mid_feathers'())
    Alpha=0.95
    GravityOverride:Vector=@n78
    ExternalForce="0,0,0"
   ) -> (Pose) {
    angularDampingOverride=1
    angularSpringConstant=20000
    bAngularSpring=true
    bOverrideAngularDamping=true
    bOverrideLinearDamping=true
    bUseGravityOverride=true
    gravityScale=0.35
    linearDampingOverride=0.95
    simulationSpace=RootRelative
    alphaBoolBlend {blendOption=Linear}
    alphaScaleBias {scale=0.5}
    boundBone {boneName=dyn_feather_08_r}
    componentAppliedLinearAccClamp=<3000 3000 3000>
    componentLinearAccScale=<0.4 0.4 0.4>
    componentLinearVelScale=<0.4 0.4 0.4>
    gravityOverride {z=-980}
    physicsBodyDefinitions [{
      boundBone {boneName=dyn_feather_08_r}
      boxExtents=<15 15 15>
      constraintSetup {angularLimitsMax=<0 15 0>, angularLimitsMin=<0 -15 0>, angularTarget=<1 0 0>}
      localJointOffset=<1.5 0 0>
     }]
   }
   n80 = Anim Dynamics - Bone: dyn_feather_09_r!(ComponentPose=@n79, Alpha=0.95, GravityOverride:Vector=@n78, ExternalForce="0,0,0") -> (Pose) {
    angularDampingOverride=1
    angularSpringConstant=20000
    bAngularSpring=true
    bOverrideAngularDamping=true
    bOverrideLinearDamping=true
    bUseGravityOverride=true
    gravityScale=0.35
    linearDampingOverride=0.95
    simulationSpace=RootRelative
    alphaBoolBlend {blendOption=Linear}
    alphaScaleBias {scale=0.5}
    boundBone {boneName=dyn_feather_09_r}
    componentAppliedLinearAccClamp=<3000 3000 3000>
    componentLinearAccScale=<0.4 0.4 0.4>
    componentLinearVelScale=<0.4 0.4 0.4>
    gravityOverride {z=-980}
    physicsBodyDefinitions [{
      boundBone {boneName=dyn_feather_09_r}
      boxExtents=<10 10 10>
      constraintSetup {angularLimitsMax=<0 10 0>, angularLimitsMin=<0 -5 0>, angularTarget=<1 0 0>}
      localJointOffset=<2 0 0>
     }]
   }
   n81 = Anim Dynamics - Bone: dyn_feather_10_r!(ComponentPose=@n80, Alpha=0.95, GravityOverride:Vector=@n78, ExternalForce="0,0,0") -> (Pose) {
    angularDampingOverride=1
    angularSpringConstant=20000
    bAngularSpring=true
    bOverrideAngularDamping=true
    bOverrideLinearDamping=true
    bUseGravityOverride=true
    gravityScale=0.35
    linearDampingOverride=0.95
    simulationSpace=RootRelative
    alphaBoolBlend {blendOption=Linear}
    alphaScaleBias {scale=0.5}
    boundBone {boneName=dyn_feather_10_r}
    componentAppliedLinearAccClamp=<3000 3000 3000>
    componentLinearAccScale=<0.4 0.4 0.4>
    componentLinearVelScale=<0.4 0.4 0.4>
    gravityOverride {z=-980}
    physicsBodyDefinitions [{
      boundBone {boneName=dyn_feather_10_r}
      boxExtents=<15 12 10>
      constraintSetup {angularLimitsMax=<0 5 0>, angularLimitsMin=<0 -10 0>, angularTarget=<1 0 0>}
      localJointOffset=<2.5 0 0>
     }]
   }
   n82 = Anim Dynamics - Bone: dyn_feather_11_r!(ComponentPose=@n81, Alpha=0.95, GravityOverride:Vector=@n78, ExternalForce="0,0,0") -> (Pose) {
    angularDampingOverride=1
    angularSpringConstant=20000
    bAngularSpring=true
    bOverrideAngularDamping=true
    bOverrideLinearDamping=true
    bUseGravityOverride=true
    gravityScale=0.35
    linearDampingOverride=0.95
    simulationSpace=RootRelative
    alphaBoolBlend {blendOption=Linear}
    alphaScaleBias {scale=0.5}
    boundBone {boneName=dyn_feather_11_r}
    componentAppliedLinearAccClamp=<3000 3000 3000>
    componentLinearAccScale=<0.4 0.4 0.4>
    componentLinearVelScale=<0.4 0.4 0.4>
    gravityOverride {z=-980}
    physicsBodyDefinitions [{
      boundBone {boneName=dyn_feather_11_r}
      boxExtents=<10 18 10>
      constraintSetup {angularLimitsMax=<0 8 0>, angularLimitsMin=<0 -8 0>, angularTarget=<1 0 0>}
      localJointOffset=<3 0 0>
     }]
   }
   n83 = Anim Dynamics - Bone: dyn_feather_12_r!(ComponentPose=@n82, Alpha=0.95, GravityOverride:Vector=@n78, ExternalForce="0,0,0") -> (Pose) {
    angularDampingOverride=1
    angularSpringConstant=20000
    bAngularSpring=true
    bOverrideAngularDamping=true
    bOverrideLinearDamping=true
    bUseGravityOverride=true
    gravityScale=0.35
    linearDampingOverride=0.95
    simulationSpace=RootRelative
    alphaBoolBlend {blendOption=Linear}
    alphaScaleBias {scale=0.5}
    boundBone {boneName=dyn_feather_12_r}
    componentAppliedLinearAccClamp=<3000 3000 3000>
    componentLinearAccScale=<0.4 0.4 0.4>
    componentLinearVelScale=<0.4 0.4 0.4>
    gravityOverride {z=-980}
    physicsBodyDefinitions [{
      boundBone {boneName=dyn_feather_12_r}
      boxExtents=<20 30 30>
      constraintSetup {angularLimitsMax=<0 15 0>, angularLimitsMin=<0 -15 0>, angularTarget=<1 0 0>}
      localJointOffset=<3 0 0>
     }]
   }
   n84 = Anim Dynamics - Bone: dyn_feather_13_r!(ComponentPose=@n83, Alpha=0.95, GravityOverride:Vector=@n78, ExternalForce="0,0,0") -> (Pose) {
    angularDampingOverride=1
    angularSpringConstant=20000
    bAngularSpring=true
    bOverrideAngularDamping=true
    bOverrideLinearDamping=true
    bUseGravityOverride=true
    gravityScale=0.35
    linearDampingOverride=0.95
    simulationSpace=RootRelative
    alphaBoolBlend {blendOption=Linear}
    alphaScaleBias {scale=0.5}
    boundBone {boneName=dyn_feather_13_r}
    componentAppliedLinearAccClamp=<3000 3000 3000>
    componentLinearAccScale=<0.4 0.4 0.4>
    componentLinearVelScale=<0.4 0.4 0.4>
    gravityOverride {z=-980}
    physicsBodyDefinitions [{
      boundBone {boneName=dyn_feather_13_r}
      boxExtents=<15 15 15>
      constraintSetup {angularLimitsMax=<0 15 0>, angularLimitsMin=<0 -15 0>, angularTarget=<1 0 0>}
      localJointOffset=<1.5 0 0>
     }]
   }
   n85 = Anim Dynamics - Bone: dyn_feather_08_l!(ComponentPose=@n84, Alpha=0.95, GravityOverride:Vector=@n78, ExternalForce="0,0,0") -> (Pose) {
    angularDampingOverride=1
    angularSpringConstant=20000
    bAngularSpring=true
    bOverrideAngularDamping=true
    bOverrideLinearDamping=true
    bUseGravityOverride=true
    gravityScale=0.35
    linearDampingOverride=0.95
    simulationSpace=RootRelative
    alphaBoolBlend {blendOption=Linear}
    alphaScaleBias {scale=0.5}
    boundBone {boneName=dyn_feather_08_l}
    componentAppliedLinearAccClamp=<3000 3000 3000>
    componentLinearAccScale=<0.4 0.4 0.4>
    componentLinearVelScale=<0.4 0.4 0.4>
    gravityOverride {z=-980}
    physicsBodyDefinitions [{
      boundBone {boneName=dyn_feather_08_l}
      boxExtents=<15 15 15>
      constraintSetup {angularLimitsMax=<0 15 0>, angularLimitsMin=<0 -15 0>, angularTarget=<1 0 0>}
      localJointOffset=<2 0 0>
     }]
   }
   n86 = Anim Dynamics - Bone: dyn_feather_09_l!(ComponentPose=@n85, Alpha=0.95, GravityOverride:Vector=@n78, ExternalForce="0,0,0") -> (Pose) {
    angularDampingOverride=1
    angularSpringConstant=20000
    bAngularSpring=true
    bOverrideAngularDamping=true
    bOverrideLinearDamping=true
    bUseGravityOverride=true
    gravityScale=0.35
    linearDampingOverride=0.95
    simulationSpace=RootRelative
    alphaBoolBlend {blendOption=Linear}
    alphaScaleBias {scale=0.5}
    boundBone {boneName=dyn_feather_09_l}
    componentAppliedLinearAccClamp=<3000 3000 3000>
    componentLinearAccScale=<0.4 0.4 0.4>
    componentLinearVelScale=<0.4 0.4 0.4>
    gravityOverride {z=-980}
    physicsBodyDefinitions [{
      boundBone {boneName=dyn_feather_09_l}
      boxExtents=<10 10 10>
      constraintSetup {angularLimitsMax=<0 10 0>, angularLimitsMin=<0 -5 0>, angularTarget=<1 0 0>}
      localJointOffset=<2 0 0>
     }]
   }
   n87 = Anim Dynamics - Bone: dyn_feather_10_l!(ComponentPose=@n86, Alpha=0.95, GravityOverride:Vector=@n78, ExternalForce="0,0,0") -> (Pose) {
    angularDampingOverride=1
    angularSpringConstant=20000
    bAngularSpring=true
    bOverrideAngularDamping=true
    bOverrideLinearDamping=true
    bUseGravityOverride=true
    gravityScale=0.35
    linearDampingOverride=0.95
    simulationSpace=RootRelative
    alphaBoolBlend {blendOption=Linear}
    alphaScaleBias {scale=0.5}
    boundBone {boneName=dyn_feather_10_l}
    componentAppliedLinearAccClamp=<3000 3000 3000>
    componentLinearAccScale=<0.4 0.4 0.4>
    componentLinearVelScale=<0.4 0.4 0.4>
    gravityOverride {z=-980}
    physicsBodyDefinitions [{
      boundBone {boneName=dyn_feather_10_l}
      boxExtents=<15 12 10>
      constraintSetup {angularLimitsMax=<0 5 0>, angularLimitsMin=<0 -10 0>, angularTarget=<1 0 0>}
      localJointOffset=<2.5 0 0>
     }]
   }
   n88 = Anim Dynamics - Bone: dyn_feather_11_l!(ComponentPose=@n87, Alpha=0.95, GravityOverride:Vector=@n78, ExternalForce="0,0,0") -> (Pose) {
    angularDampingOverride=1
    angularSpringConstant=20000
    bAngularSpring=true
    bOverrideAngularDamping=true
    bOverrideLinearDamping=true
    bUseGravityOverride=true
    gravityScale=0.35
    linearDampingOverride=0.95
    simulationSpace=RootRelative
    alphaBoolBlend {blendOption=Linear}
    alphaScaleBias {scale=0.5}
    boundBone {boneName=dyn_feather_11_l}
    componentAppliedLinearAccClamp=<3000 3000 3000>
    componentLinearAccScale=<0.4 0.4 0.4>
    componentLinearVelScale=<0.4 0.4 0.4>
    gravityOverride {z=-980}
    physicsBodyDefinitions [{
      boundBone {boneName=dyn_feather_11_l}
      boxExtents=<10 18 10>
      constraintSetup {angularLimitsMax=<0 8 0>, angularLimitsMin=<0 -8 0>, angularTarget=<1 0 0>}
      localJointOffset=<3 0 0>
     }]
   }
   n89 = Anim Dynamics - Bone: dyn_feather_12_l!(ComponentPose=@n88, Alpha=0.95, GravityOverride:Vector=@n78, ExternalForce="0,0,0") -> (Pose) {
    angularDampingOverride=1
    angularSpringConstant=20000
    bAngularSpring=true
    bOverrideAngularDamping=true
    bOverrideLinearDamping=true
    bUseGravityOverride=true
    gravityScale=0.35
    linearDampingOverride=0.95
    simulationSpace=RootRelative
    alphaBoolBlend {blendOption=Linear}
    alphaScaleBias {scale=0.5}
    boundBone {boneName=dyn_feather_12_l}
    componentAppliedLinearAccClamp=<3000 3000 3000>
    componentLinearAccScale=<0.4 0.4 0.4>
    componentLinearVelScale=<0.4 0.4 0.4>
    gravityOverride {z=-980}
    physicsBodyDefinitions [{
      boundBone {boneName=dyn_feather_12_l}
      boxExtents=<20 30 30>
      constraintSetup {angularLimitsMax=<0 15 0>, angularLimitsMin=<0 -15 0>, angularTarget=<1 0 0>}
      localJointOffset=<3 0 0>
     }]
   }
   n90 = Anim Dynamics - Bone: dyn_feather_13_l!(ComponentPose=@n89, Alpha=0.95, GravityOverride:Vector=@n78, ExternalForce="0,0,0") -> (Pose) {
    angularDampingOverride=1
    angularSpringConstant=20000
    bAngularSpring=true
    bOverrideAngularDamping=true
    bOverrideLinearDamping=true
    bUseGravityOverride=true
    gravityScale=0.35
    linearDampingOverride=0.95
    simulationSpace=RootRelative
    alphaBoolBlend {blendOption=Linear}
    alphaScaleBias {scale=0.5}
    boundBone {boneName=dyn_feather_13_l}
    componentAppliedLinearAccClamp=<3000 3000 3000>
    componentLinearAccScale=<0.4 0.4 0.4>
    componentLinearVelScale=<0.4 0.4 0.4>
    gravityOverride {z=-980}
    physicsBodyDefinitions [{
      boundBone {boneName=dyn_feather_13_l}
      boxExtents=<15 15 15>
      constraintSetup {angularLimitsMax=<0 15 0>, angularLimitsMin=<0 -15 0>, angularTarget=<1 0 0>}
      localJointOffset=<1.5 0 0>
     }]
   }
   Save cached pose 'CRCache_Large_feathers'(Pose=Component To Local(ComponentPose=@n90)) {cacheName=CRCache_Large_feathers}
  ]}
 EventGraph {nodes [
   Event BlueprintUpdateAnimation()
   > n1 = Sequence!(execute) -> (
    then_0
    then_1
    then_2:Exec
    then_3:Exec
    then_4:Exec
    then_5:Exec
    then_6:Exec
    then_7:Exec
    then_8:Exec
    then_9:Exec
    then_10:Exec
    then_11:Exec
   )
   Set SkydiveFloatingAlpha(execute=@n1.then_11, SkydiveFloatingAlpha:float="To Float (Boolean)"(InBool=Get bIsSkydiveFloating())) -> (Output_Get:float)
   n2 = GetGameTimeInSeconds()
   n3 = RandomRotatorFunction!(
    execute=@n1.then_5
    GameTimeInSeconds:double=@n2
    PerlinScale:double=1.9
    Frequency:Rotator="15,0,0"
    Amplitude:Rotator="1.5,0,0"
    Offset:Rotator="0, 0, 0"
    TimeOffset:double=1
   ) -> (then, RandomRotator:Rotator)
   > Set ambientRotatorA(ambientRotatorA:Rotator=@n3) -> (Output_Get:Rotator)
   n4 = RandomRotatorFunction(execute=@n1.then_7, GameTimeInSeconds=@n2, PerlinScale=1.7, Frequency="20,0,0", Amplitude="1.3,0,0", Offset="0, 0, 0", TimeOffset=0)
   > Set ambientRotatorC(ambientRotatorC:Rotator=@n4) -> (Output_Get:Rotator)
   n5 = RandomRotatorFunction(execute=@n1.then_8, GameTimeInSeconds=@n2, PerlinScale=2, Frequency="15,0,0", Amplitude="1.2,0,0", Offset="0, 0, 0", TimeOffset=0)
   > Set ambientRotatorD(ambientRotatorD:Rotator=@n5) -> (Output_Get:Rotator)
   n6 = Gravity Override State Function DEPRECATED(
    execute=@n1.then_3
    CharacterPartAnimInstance=Self-Reference()
    GameTimeInSeconds=GetGameTimeInSeconds()
    GravityOverrideParams=Get Gravity Override Params_0() -> (Gravity Override Params_0:Gravity Override Params Struct Structure)
   )
   > Set GravityOverride(GravityOverride=@n6)
   n7 = Branch(execute=@n1.then_2, Condition=Get bIsParachuteOpened())
   > Set parachuteOpen(parachuteOpen:double=0) -> (Output_Get:double)
   Set parachuteOpen(execute=@n7.else, parachuteOpen=1)
   Set skyDive(execute=@n1.then_1, skyDive:bool=OR Boolean(A=Get bIsSkyDiving(), B=Get bIsSkydiveFloating())) -> (Output_Get:bool)
   Set runningVar(
    execute=@n1.then_0
    runningVar:bool=OR Boolean!(A=Get bIsCrouchSprinting(), B=Get bIsTacticalSprinting(), C:bool=Get bIsCrouching(), D:bool=Get bIsSprinting(), E:bool=Get bIsOnGrindRail()) -> (ReturnValue)
   ) -> (Output_Get:bool)
   n8 = RandomRotatorFunction(execute=@n1.then_6, GameTimeInSeconds=@n2, PerlinScale=2.75, Frequency="10,5,0", Amplitude="1.7,0,0", Offset="0, 0, 0", TimeOffset=1)
   > Set ambienRotatorB(ambienRotatorB:Rotator=@n8) -> (Output_Get:Rotator)
   n9 = GetRealTimeSeconds()
   n10 = "Sin (Radians)"(A=float * float(A=float * float(A=1.2, B=5), B=@n9))
   n11 = "Sin (Radians)"(A=float * float(A=float * float(A=1.1, B=1.1), B=@n9))
   n12 = "Sin (Radians)"(A=float * float(A=float * float(A=0.9, B=1.1), B=@n9))
   Set WingsRotation(
    execute=@n1.then_10
    WingsRotation:Rotator=MakeRotator(
     Roll=float * float(A=@n11, B=0.15)
     Pitch=float * float(A=@n12, B=0.3)
     Yaw=Remap Float(Input=@n10, OldRange=(X=-1,Y=1), NewRange=(X=-7,Y=7)) -> (Output=~)
    )
   ) -> (Output_Get:Rotator)
   n13 = Branch(execute=@n1.then_4, Condition=OR Boolean!(A=Get bIsSkyDiving(), B=Get bIsSkydiveDiveMode(), C=Get bIsSkydiveFloating()) -> (ReturnValue))
   > Set skyDiveMode(skyDiveMode:double=float * float(A=0.5, B=0.45) {is=title K2Node_PromotableOperator}) -> (Output_Get:double)
   Set skyDiveMode(execute=@n13.else, skyDiveMode=float * float(A=0.5, B=0.95))
   n14 = Branch(execute=@n1.then_9, Condition=Get bIsJumping())
   > Set fakeMovement(fakeMovement:double=4.75) -> (Output_Get:double)
   Set fakeMovement(execute=@n14.else, fakeMovement=0)
   > n15 = Branch(Condition=OR Boolean!(A=Get bIsParachuteOpened(), B=Get bIsSkydiveDiveMode(), C=Get bIsSkydiveFloating(), D:bool=Get bIsSkyDiving()) -> (ReturnValue))
   > Set fakeMovement(fakeMovement=0.28)
   Set fakeMovement(execute=@n15.else, fakeMovement=0)
   TryGetPawnOwner()
   Get Gravity Override Params()
  ]}
]
variables [
 ambienRotatorB {type=Rotator}
 ambientRotatorA {type=Rotator}
 ambientRotatorC {type=Rotator}
 ambientRotatorD {type=Rotator}
 fakeMovement {type=number}
 Gravity Override Params {type=GravityOverrideParamsStruct}
 Gravity Override Params_0 {type=GravityOverrideParamsStruct, default {skydive_FloatingDown {windAmplitude=<200 50 400>, windFrequency=<1.5 14 8>}}}
 GravityOverride {type=Vector}
 idleVar {type=number}
 parachuteOpen {type=number}
 Ring_A_Float {type=Vector}
 runningVar {type=boolean}
 skyDive {type=boolean}
 SkydiveFloatingAlpha {type=number, default=1}
 skyDiveMode {type=number}
 Wing_L_Float {type=Vector}
 Wing_R_Float {type=Vector}
 WingsRotation {type=Rotator}
]
