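# Physics asset for the MillionaireGem backpack cosmetic, named for the F_MED (female, medium) pack mesh.
# Its constraint list is empty, so it defines no joints between physics bodies.
# No bodies, collision shapes or simulation settings appear, so it shows no jiggle or dangle setup of its own.
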PhysicsAsset /BRCosmetics/Accessories/FORT_Backpacks/Backpack_F_MED_MillionaireGem/Meshes/F_MED_MillionaireGem_Pack_Physics
constraints []
